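# Sound cue for the weak-point hit impact of the TrifleMale cosmetic pickaxe.
# A short fixed delay feeds a modulator with randomized pitch and fixed volume, then an evenly weighted random pick among six "TrifleLightningMale Reactive" waves.
# It uses player pickaxe impact attenuation, a shared cosmetic impact property sheet and the impacts sound class.
# A per-owner concurrency override caps simultaneous plays, stopping the oldest.
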SoundCue /BRCosmetics/Sounds/PickAxes/TrifleMale/PickaxeImpactWeakPoint_TrifleMale_Cue
root=Delay_0
nodes [
 {id=Delay_0, children [Modulator_0], properties {delayMax=0.1, delayMin=0.1}}
 {id=Modulator_0, children [Random_0], properties {volumeMax=1, volumeMin=1}}
 {id=Random_0, children [WavePlayer_0,WavePlayer_1,WavePlayer_2,WavePlayer_3,WavePlayer_4,WavePlayer_5], properties {weights [1,1,1,1,1,1]}}
 {id=WavePlayer_0, properties {soundWaveAssetPtr=&/Pickaxe_TrifleLightningMale_Reactive_01}}
 {id=WavePlayer_1, properties {soundWaveAssetPtr=&/Pickaxe_TrifleLightningMale_Reactive_02}}
 {id=WavePlayer_2, properties {soundWaveAssetPtr=&/Pickaxe_TrifleLightningMale_Reactive_03}}
 {id=WavePlayer_3, properties {soundWaveAssetPtr=&/Pickaxe_TrifleLightningMale_Reactive_04}}
 {id=WavePlayer_4, properties {soundWaveAssetPtr=&/Pickaxe_TrifleLightningMale_Reactive_05}}
 {id=WavePlayer_5, properties {soundWaveAssetPtr=&/Pickaxe_TrifleLightningMale_Reactive_06}}
]
properties {
 attenuationSettings=/Game/Sounds/Attenuation/Fort_Weapon_Pickaxe_Player_Impact_Att
 audioPropertiesSheet=/BRCosmetics/Sounds/Settings/PropertySheets/Pickaxe/APS_Cosmetic_Pickaxe_Impact
 bOverrideConcurrency=true
 maxDistance=3512
 soundClassObject=&Fort_Impacts
 volumeMultiplier=0.375
 concurrencyOverrides {bLimitToOwner=true, maxCount=2, resolutionRule=StopOldest}
 concurrencySet [null]
}
